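# Octane Valkyrie car skin material: a Fresnel-shaded two-color paint with an animated emissive-style pattern.
# FinishColor blends toward TeamColor by Fresnel, then picks up faked highlights from a tangent-space camera dot product and a reflection-vector lookup into an environment texture.
# The skin texture's red channel masks a panning gradient stripe and also drives its V coordinate, so the stripe sweeps through the pattern.
# That glow, tinted by FinishColor and scaled by EmissiveIntensity, is added over the paint.
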Material /VehicleCosmetics/Bodies/Octane/Skins/OctaneValkyrie/MAT_Skin_Octane_Valkyrie
graph {
 properties {bAutomaticallySetUsageInEditor=false, usedShadingModels=Default Lit}
 nodes [
  n1 = VectorParameter() {parameterName=FinishColor, defaultValue=<rgba 0.5 0.5 0.5 1>}
  n2 = Fresnel() {baseReflectFraction=0, exponent=3}
  n3 = ConstantBiasScale(None=@n2) {bias=-0.01, scale=1.2}
  n4 = ConstantBiasScale(None=@n2) {bias=-0.5, scale=8}
  n5 = Fresnel() {baseReflectFraction=0, exponent=0.5}
  n6 = Clamp(None=ConstantBiasScale(None=OneMinus(None=@n5)) {bias=-0.125, scale=3}) {maxDefault=0.8, minDefault=0.3}
  n7 = Multiply(
   A=LinearInterpolate(A=@n1.RGB, B=VectorParameter() {parameterName=TeamColor, defaultValue=<rgba 1 0 0 1>}.RGB, Alpha=Saturate(None=@n3))
   B=Saturate(None=Max(A=@n4, B=@n6))
  )
  n8 = Transform(None=CameraVectorWS()) {transformSourceType=TRANSFORMSOURCE_World, transformType=TRANSFORM_Tangent}
  n9 = Constant3Vector() {constant=<rgba -0.1 1.9 1 0>}
  n10 = Power(Base=ConstantBiasScale(None=DotProduct(A=@n8, B=@n9)) {bias=-0.9, scale=1}) {constExponent=1}
  n11 = ConstantBiasScale(None=@n10) {bias=-0.5, scale=0.2}
  n12 = ConstantBiasScale(None=@n10) {bias=-0.99, scale=10}
  n13 = ReflectionVectorWS()
  n14 = TextureSample(UVs=ComponentMask(None=ConstantBiasScale(None=@n13)) {g=true, r=true}) {texture=/VehicleCosmetics/Textures/UniversalTextures/T_ENVPack}
  n15 = LinearInterpolate(A=@n14.R, B=@n14.R) {constAlpha=0.4}
  n16 = Multiply(A=@n7, B=Add(A=Max(A=@n11, B=@n12), B=ConstantBiasScale(None=@n15) {bias=-0.074, scale=3}))
  Add(A=@n7, B=@n16)
  n17 = ComponentMask(None=TextureCoordinate()) {r=true}
  n18 = TextureSample() {texture=&/T_valkyrieRemake_RGB}
  n19 = Multiply(A=@n18.R) {constB=8}
  n20 = TextureSample(UVs=Panner(Coordinate=AppendVector(A=@n17, B=@n19), Time=Time() {bOverride_Period=true, period=60}) {speedY=1}) {texture=/VehicleCosmetics/Textures/UniversalTextures/T_GradientStripe}
  n21 = Saturate(None=ConstantBiasScale(None=Multiply(A=@n20.R, B=@n18.R)) {bias=-0.3, scale=2})
  n22 = ScalarParameter() {defaultValue=2.4, parameterName=EmissiveIntensity}
  n23 = Multiply(A=@n7, B=OneMinus(None=Fresnel() {baseReflectFraction=0.5, exponent=0.1}))
  Add(A=Multiply(A=Multiply(A=@n21, B=@n22), B=@n1.RGB), B=Saturate(None=Add(A=@n23, B=@n16)))
 ]
}
