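# Material function for the Valkyrie skin of the Octane car body, producing its color and glow as material attributes.
# Fresnel terms blend FinishColor toward TeamColor and shade it, with a faked view-angle highlight and reflection-vector environment texture added to base color.
# Emissive combines that Fresnel-shaded color and a TeamColor rim with FinishColor pulses, a time-panned triangle wave offset and masked by the skin texture's red channel.
# The gradient stripe sample is unconnected.
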MaterialFunction /VehicleCosmetics/Bodies/Octane/Skins/OctaneValkyrie/MF_Skin_Valkyrie
graph {nodes [
  n1 = VectorParameter() {parameterName=FinishColor, defaultValue=<rgba 0.5 0.5 0.5 1>}
  n2 = VectorParameter() {parameterName=TeamColor, defaultValue=<rgba 1 0 0 1>}
  n3 = Fresnel() {baseReflectFraction=0, exponent=3}
  n4 = ConstantBiasScale(None=@n3) {bias=-0.01, scale=1.2}
  n5 = ConstantBiasScale(None=@n3) {bias=-0.5, scale=8}
  n6 = Fresnel() {baseReflectFraction=0, exponent=0.5}
  n7 = Clamp(None=ConstantBiasScale(None=OneMinus(None=@n6)) {bias=-0.125, scale=3}) {maxDefault=0.8, minDefault=0.3}
  n8 = Multiply(A=LinearInterpolate(A=@n1.RGB, B=@n2.RGB, Alpha=Saturate(None=@n4)), B=Saturate(None=Max(A=@n5, B=@n7)))
  n9 = Transform(None=CameraVectorWS()) {transformSourceType=TRANSFORMSOURCE_World, transformType=TRANSFORM_Tangent}
  n10 = Constant3Vector() {constant=<rgba -0.1 1.9 1 0>}
  n11 = Power(Base=ConstantBiasScale(None=DotProduct(A=@n9, B=@n10)) {bias=-0.9, scale=1}) {constExponent=1}
  n12 = ReflectionVectorWS()
  n13 = TextureSample(UVs=ComponentMask(None=ConstantBiasScale(None=@n12)) {g=true, r=true}) {texture=/VehicleCosmetics/Textures/UniversalTextures/T_ENVPack}
  n14 = Add(
   A=Max(A=ConstantBiasScale(None=@n11) {bias=-0.5, scale=0.2}, B=ConstantBiasScale(None=@n11) {bias=-0.99, scale=10})
   B=ConstantBiasScale(None=LinearInterpolate(A=@n13.R, B=@n13.R) {constAlpha=0.4}) {bias=-0.074, scale=3}
  )
  n15 = TextureCoordinate()
  n16 = TextureSample() {texture=&/T_valkyrieRemake_RGB}
  n17 = Panner(
   Coordinate=AppendVector(A=ComponentMask(None=@n15) {r=true}, B=Multiply(A=@n16.R) {constB=8})
   Time=Time() {bOverride_Period=true, period=60}
  ) {speedY=1}
  n18 = ComponentMask(None=@n17) {g=true}
  n19 = ConstantBiasScale(None=Subtract(A=@n18, B=Floor(None=@n18))) {bias=-0.5, scale=2}
  n20 = Multiply(A=Saturate(None=Abs(None=@n19)), B=@n16.R)
  n21 = Multiply(A=Saturate(None=ConstantBiasScale(None=@n20) {bias=-0.35, scale=2}), B=ScalarParameter() {defaultValue=10, parameterName=EmissiveIntensity})
  n22 = OneMinus(None=Fresnel() {baseReflectFraction=0.5, exponent=0.1})
  n23 = Saturate(None=ConstantBiasScale(None=@n3) {bias=-0.3, scale=3})
  n24 = Saturate(None=Add(A=Multiply(A=@n8, B=@n22), B=Multiply(A=@n2.RGB, B=@n23)))
  n25 = SetMaterialAttributes!(
   MaterialAttributes
   BaseColor=Add(A=Multiply(A=@n8) {constB=0.75}, B=Multiply(A=@n8, B=@n14))
   Metallic
   Roughness
   EmissiveColor=Add(A=Multiply(A=@n21, B=@n1.RGB), B=Multiply(A=@n24) {constB=0.5})
   Normal
   AmbientOcclusion
  ) {attributeSetTypes [
    #1
    #2
    #3
    #4
    #5
    #6
   ]}
  FunctionOutput(None=@n25)
  TextureSample(UVs=@n17) {texture=/VehicleCosmetics/Textures/UniversalTextures/T_GradientStripe}
 ]}
